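# Sound cue for the music of the Taco Time emote in Athena, using the first-person emote music sound class.
# A short fixed delay feeds a wave player that loops a single clean music loop.
# An emote music concurrency set is assigned, and concurrency override values specify stopping the oldest instance.
# The cue also lists a send to the Fortnite music audio-analysis submix.
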SoundCue /Game/Athena/Sounds/Emotes/Taco_Time/Emote_Taco_Time_Music_Cue
root=Delay_0
nodes [
 {id=Delay_0, children [WavePlayer_0], properties {delayMax=0.1, delayMin=0.1}}
 {id=WavePlayer_0, properties {bLooping=true, soundWaveAssetPtr=&/Emote_GG_TT_MX_Loop_01_Clean}}
]
properties {
 maxDistance=2097152
 soundClassObject=/Game/Sounds/FortSoundClassesAndModes/Fort_Music_Emote_1P_PSM
 volumeMultiplier=1
 concurrencyOverrides {maxCount=3, resolutionRule=StopOldest}
 concurrencySet [/Game/Sounds/Concurrency_Rules/Athena_EmoteMusic_1P_Conc]
 soundSubmixSends [{minSendLevel=0, soundSubmix=&Fortnite_AudioAnalysis_Music}]
}
